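# Animation Blueprint for the player's physics-flying wingsuit, implementing linked anim layers on the Fortnite player skeleton.
# The SecondMovementMode final-pose override instantly switches between the incoming cached pose and a pose copied from another skeletal mesh when bWingsuitOn is true, then applies the TDiA.Custom_01 slot.
# ItemFallLandAdditive forces the BothHandFK curve on.
# The main AnimGraph, EventGraph and remaining layers are unwired.
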AnimBlueprint /WingSuit/Animation/Game/MainPlayer/AnimAssets/Wingsuit_PhysicsFlying_AnimBP
target_skeleton=&Fortnite_M_Avg_Player_Skeleton
graphs [
 AnimGraph {nodes [Output Pose()]}
 EventGraph {links [], nodes []}
 ItemAimOffset {nodes [
   Input Pose!() -> (Pose, AimOffsetAlpha:double=~, Yaw:double=~, Pitch:double=~) {
    name=InPose_AimOffset
    inputs [
     AimOffsetAlpha {type {
       bIsConst=false
       bIsReference=false
       bIsUObjectWrapper=false
       bIsWeakPointer=false
       bSerializeAsSinglePrecisionFloat=false
       containerType=None
       pinCategory=real
       pinSubCategory=double
       pinSubCategoryObject=null
       pinSubCategoryMemberReference {memberGuid=#1, memberName=None, memberParent=null}
       pinValueType {
        bTerminalIsConst=false
        bTerminalIsUObjectWrapper=false
        bTerminalIsWeakPointer=false
        terminalCategory=None
        terminalSubCategory=None
        terminalSubCategoryObject=null
       }
      }}
     Yaw {type {
       bIsConst=false
       bIsReference=false
       bIsUObjectWrapper=false
       bIsWeakPointer=false
       bSerializeAsSinglePrecisionFloat=false
       containerType=None
       pinCategory=real
       pinSubCategory=double
       pinSubCategoryObject=null
       pinSubCategoryMemberReference {memberGuid=#1, memberName=None, memberParent=null}
       pinValueType {
        bTerminalIsConst=false
        bTerminalIsUObjectWrapper=false
        bTerminalIsWeakPointer=false
        terminalCategory=None
        terminalSubCategory=None
        terminalSubCategoryObject=null
       }
      }}
     Pitch {type {
       bIsConst=false
       bIsReference=false
       bIsUObjectWrapper=false
       bIsWeakPointer=false
       bSerializeAsSinglePrecisionFloat=false
       containerType=None
       pinCategory=real
       pinSubCategory=double
       pinSubCategoryObject=null
       pinSubCategoryMemberReference {memberGuid=#1, memberName=None, memberParent=null}
       pinValueType {
        bTerminalIsConst=false
        bTerminalIsUObjectWrapper=false
        bTerminalIsWeakPointer=false
        terminalCategory=None
        terminalSubCategory=None
        terminalSubCategoryObject=null
       }
      }}
    ]
   }
   Input Pose!() -> (Pose) {name=InPose_UpperLowerPreMeleeAO}
   Output Pose() {layerGroup=ItemOverride}
  ]}
 ItemFallAdditive {nodes [
   Input Pose() {name=InPoseItemFallAdditive}
   Output Pose() {layerGroup=ItemOverride}
  ]}
 ItemFallLandAdditive {nodes [Output Pose(Result=Modify Curve(SourcePose=Input Pose() {name=InPoseItemFallLandAdditive}, CurveValues_0=1, Alpha=1) {curveNames [BothHandFK], curveValues [0]}) {layerGroup=ItemOverride}]}
 ItemFallLevel2Override {nodes [
   Input Pose() {name=InPose_FallLevel2}
   Output Pose() {layerGroup=ItemOverride}
  ]}
 ItemFinalPoseOverride {nodes [
   Input Pose() {name=InFinalPose}
   Output Pose() {layerGroup=ItemOverride}
  ]}
 ItemFullBodyOverride {nodes [
   Input Pose() {name=InPoseFullBody}
   Output Pose() {layerGroup=ItemOverride}
  ]}
 ItemIdleOverride {nodes [
   Input Pose() {name=InPose_Idle}
   Output Pose() {layerGroup=ItemOverride}
  ]}
 ItemJetPackJumpAdditive {nodes [Output Pose() {layerGroup=ItemOverride}]}
 ItemJetPackStartAdditive {nodes [Output Pose() {layerGroup=ItemOverride}]}
 ItemJumpLoopAdditive {nodes [
   Input Pose() {name=InPoseItemJumpLoopAdditive}
   Output Pose() {layerGroup=ItemOverride}
  ]}
 ItemJumpOverride {nodes [
   Input Pose() {name=InPoseJump}
   Output Pose() {layerGroup=ItemOverride}
  ]}
 ItemJumpUpAdditive {nodes [
   Input Pose() {name=InPoseItemJumpUpAdditive}
   Output Pose() {layerGroup=ItemOverride}
  ]}
 ItemPreAimOffsets {nodes [
   Input Pose() {name=InPosePreAimOffsets}
   Output Pose() {layerGroup=ItemOverride}
  ]}
 ItemPreIK {nodes [
   Input Pose() {name=InPose_PreIK}
   Output Pose() {layerGroup=ItemOverride}
  ]}
 ItemRelaxedEntry {nodes [
   Input Pose() {name=InPoseRelaxedEntry}
   Output Pose() {layerGroup=ItemOverride}
  ]}
 ItemSkeletalControl {nodes [
   Input Pose()
   Output Pose() {layerGroup=ItemOverride}
  ]}
 ItemSwimJumpFallLoop {nodes [Output Pose() {layerGroup=ItemOverride}]}
 ItemSwimJumpSurfaceEnd {nodes [Output Pose() {layerGroup=ItemOverride}]}
 ItemSwimJumpSurfaceLoop {nodes [Output Pose() {layerGroup=ItemOverride}]}
 ItemUpperBody {nodes [
   Input Pose!() -> (Pose, InputParam:Fort Anim Input Adjusted Aim Structure) {
    name=InPoseUpperBody
    inputs [InputParam {type {
       bIsConst=false
       bIsReference=false
       bIsUObjectWrapper=false
       bIsWeakPointer=false
       bSerializeAsSinglePrecisionFloat=false
       containerType=None
       pinCategory=struct
       pinSubCategory=None
       pinSubCategoryObject=/Script/FortniteGame.FortAnimInput_AdjustedAim
       pinSubCategoryMemberReference {memberGuid=#1, memberName=None, memberParent=null}
       pinValueType {
        bTerminalIsConst=false
        bTerminalIsUObjectWrapper=false
        bTerminalIsWeakPointer=false
        terminalCategory=None
        terminalSubCategory=None
        terminalSubCategoryObject=null
       }
      }}]
   }
   Output Pose() {layerGroup=ItemOverride}
  ]}
 ItemZipLineStartAdditive {nodes [Output Pose() {layerGroup=ItemOverride}]}
 MovementMode_AimOffset {nodes [
   Input Pose!() -> (Pose) {name=InAimOffsetsPose}
   Input Pose() {name=InSourcePosePreAimOffsets}
   Output Pose() {layerGroup=MovementMode}
  ]}
 MovementMode_BaseLocomotion {nodes [
   Input Pose() {name=InPoseBaseLocomotion}
   Output Pose() {layerGroup=MovementMode}
  ]}
 MovementMode_FinalPoseOverride {nodes [
   Input Pose() {name=InFinalPose}
   Output Pose() {layerGroup=MovementMode}
  ]}
 MovementMode_FullBodyOverride {nodes [
   Input Pose() {name=InputPoseFullBody}
   Output Pose() {layerGroup=MovementMode}
  ]}
 MovementMode_InAirOverride {nodes [
   Input Pose() {name=InSource_Pose}
   Input Pose() {name=InAirPose}
   Output Pose() {layerGroup=MovementMode}
  ]}
 MovementMode_LocomotionAdditives {nodes [
   Input Pose() {name=InLocomotionAdditivesPose}
   Output Pose() {layerGroup=MovementMode}
  ]}
 MovementMode_LowerBodyOverride {nodes [
   Input Pose() {name=InLowerBodyPose}
   Input Pose() {name=InSourcePose}
   Output Pose() {layerGroup=MovementMode}
  ]}
 MovementMode_OffScreen {nodes [
   Input Pose() {name=InPoseOffscreen}
   Output Pose() {layerGroup=MovementMode}
  ]}
 MovementMode_SourcePose {nodes [
   Input Pose() {name=InSourcePoseBase}
   Output Pose() {layerGroup=MovementMode}
  ]}
 SecondMovementMode_FinalPoseOverride {nodes [
   n1 = Copy Pose From Mesh(SourceMeshComponent=Get SkeletalMesh() -> (SkeletalMesh:Skeletal Mesh Component Object Reference))
   n2 = Use cached pose 'SecondMM Pose'()
   n3 = Get bWingsuitOn() -> (bWingsuitOn:bool)
   Output Pose(Result=Slot 'TDiA.Custom_01'(Source=Blend Poses by bool(BlendPose_0=@n1, BlendPose_1=@n2, BlendTime_0=0, bActiveValue=@n3)) {slotName=TDiA.Custom_01}) {layerGroup=SecondMovementMode}
   Save cached pose 'SecondMM Pose'(Pose=Input Pose() {name=InFinalPose}) {cacheName=SecondMM Pose}
  ]}
]
variables [
 StartedLaunch {type=boolean}
 WingsuitMMETag {type=boolean}
]
