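# Player wingsuit animation Blueprint that overrides the final pose through the SecondMovementMode layer while flying.
# A flying state machine covers launch, idle, pitch/yaw transitions, flips and predictive landing, selecting launch clips by Chooser and pitch clips on becoming relevant.
# Roll/pitch additive blendspaces layer on top, with states zeroing alpha curves to suppress them.
# A second state machine lowers the pelvis during flight except predictive landing, before a montage slot.
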AnimBlueprint /WingSuit/Animation/Game/MainPlayer/AnimAssets/Wingsuit_Player_AnimBP
target_skeleton=&Fortnite_M_Avg_Player_Skeleton
graphs [
 AnimGraph {nodes [Output Pose()]}
 EventGraph {links [], nodes []}
 ItemAimOffset {nodes [
   Input Pose!() -> (Pose, AimOffsetAlpha:double, Yaw:double, Pitch:double) {name=InPose_AimOffset}
   Input Pose!() -> (Pose) {name=InPose_PreAimOffsets}
   Output Pose() {layerGroup=Item Override}
  ]}
 ItemFallAdditive {nodes [
   Input Pose() {name=InPoseItemFallAdditive}
   Output Pose() {layerGroup=Item Override}
  ]}
 ItemFallLandAdditive {nodes [
   Input Pose() {name=InPoseItemFallLandAdditive}
   Output Pose() {layerGroup=Item Override}
  ]}
 ItemFallLevel2Override {nodes [
   Input Pose() {name=InPose_FallLevel2}
   Output Pose() {layerGroup=Item Override}
  ]}
 ItemFinalPoseOverride {nodes [
   Input Pose() {name=InFinalPose}
   Output Pose() {layerGroup=Item Override}
  ]}
 ItemFullBodyAimOffsets {nodes [
   Input Pose!() -> (Pose, FullBody_AimOffsetAlpha:double, Aim_Yaw:double, Aim_Pitch:double) {name=InPoseFullBodyAimOffsets}
   Input Pose!() -> (Pose) {name=InPosePreFullBodyAimOffsets}
   Output Pose() {layerGroup=Item Override}
  ]}
 ItemFullBodyOverride {nodes [
   Input Pose() {name=InPoseFullBody}
   Output Pose() {layerGroup=Item Override}
  ]}
 ItemIdleOverride {nodes [
   Input Pose() {name=InPose_Idle}
   Output Pose() {layerGroup=Item Override}
  ]}
 ItemJetPackJumpAdditive {nodes [Output Pose() {layerGroup=Item Override}]}
 ItemJetPackStartAdditive {nodes [Output Pose() {layerGroup=Item Override}]}
 ItemJumpLoopAdditive {nodes [
   Input Pose() {name=InPoseItemJumpLoopAdditive}
   Output Pose() {layerGroup=Item Override}
  ]}
 ItemJumpOverride {nodes [
   Input Pose() {name=InPoseJump}
   Output Pose() {layerGroup=Item Override}
  ]}
 ItemJumpUpAdditive {nodes [
   Input Pose() {name=InPoseItemJumpUpAdditive}
   Output Pose() {layerGroup=Item Override}
  ]}
 ItemPreAimOffsets {nodes [
   Input Pose() {name=InPosePreAimOffsets}
   Output Pose() {layerGroup=Item Override}
  ]}
 ItemPreIK {nodes [
   Input Pose() {name=InPose_PreIK}
   Output Pose() {layerGroup=Item Override}
  ]}
 ItemRightHandIK {nodes [
   Input Pose() {name=InPosePreRightHandIK}
   Input Pose() {name=InPosePostRightHandIK}
   Output Pose() {layerGroup=Item Override}
  ]}
 ItemSkeletalControl {nodes [
   Input Pose()
   Output Pose() {layerGroup=Item Override}
  ]}
 ItemSwimJumpFallLoop {nodes [Output Pose() {layerGroup=Item Override}]}
 ItemSwimJumpSurfaceEnd {nodes [Output Pose() {layerGroup=Item Override}]}
 ItemSwimJumpSurfaceLoop {nodes [Output Pose() {layerGroup=Item Override}]}
 ItemUpperBody {nodes [
   Input Pose!() -> (Pose, InputParam:Fort Anim Input Adjusted Aim Structure) {name=InPoseUpperBody}
   Output Pose() {layerGroup=Item Override}
  ]}
 ItemZipLineStartAdditive {nodes [Output Pose() {layerGroup=Item Override}]}
 Launch_OnBecomeRelevant {nodes [
   n1 = Launch_OnBecomeRelevant!() -> (then, Context:"Anim Update Context Structure (by ref)", Node:"Anim Node Reference Structure (by ref)") {is=title K2Node_FunctionEntry}
   > n2 = Evaluate Chooser: Wingsuit_Launch_Outro_CT!(execute:Exec, Wingsuit_Gadget_AnimBP_C:Wingsuit Gadget Anim BP Object Reference=Self-Reference()) -> (execute:Exec, ChooserPlayerSettings:Chooser Player Settings Structure, Result:Anim Sequence Object Reference) {is=title K2Node_EvaluateChooser2, chooser=/WingSuit/Animation/Game/MainPlayer/AnimAssets/ChooserTable/Wingsuit_Launch_Outro_CT}
   > SetSequence!(
    execute
    SequencePlayer:"Sequence Player Reference Structure (by ref)"=Convert to Sequence Player!(Node:"Anim Node Reference Structure (by ref)"=@n1.Node) -> (SequencePlayer:Sequence Player Reference Structure, Result:bool)
    Sequence:Anim Sequence Base Object Reference=@n2.Result
   ) -> (then, ReturnValue:Sequence Player Reference Structure)
   > Return Node()
  ]}
 PitchTransitionSequence_OnBecomeRelevant {nodes [
   n1 = PitchTransitionSequence_OnBecomeRelevant!() -> (then, Context:"Anim Update Context Structure (by ref)", Node:"Anim Node Reference Structure (by ref)") {is=title K2Node_FunctionEntry}
   > Set SequencePlayer(SequencePlayer:Sequence Player Reference Structure=Convert to Sequence Player(Node=@n1.Node)) -> (Output_Get:Sequence Player Reference Structure)
   > Set bIsUpDirection(bIsUpDirection:bool="float > float"(A=Get PitchAngleBetweenDelta() -> (PitchAngleBetweenDelta:double))) -> (Output_Get:bool)
   n2 = NOT Boolean(A=Get bIsUpDirection() -> (bIsUpDirection:bool))
   n3 = "float > float"(A="Absolute (Float)"(A=Get PitchAngleBetween() -> (PitchAngleBetween:double)), B=0.4)
   n4 = Get!() -> (LocalVelocity_X:double, LocalVelocity_Y:double, LocalVelocity_Z:double)
   n5 = Get PitchAngleBetween()
   n6 = OR Boolean(
    A="float > float"(A="Absolute (Float)"(A=@n5), B=0.7)
    B="float > float"(A=Vector Length!(A_X:double=@n4, A_Y:double=@n4.LocalVelocity_Y, A_Z:double=@n4.LocalVelocity_Z) -> (ReturnValue), B=2000)
   )
   n7 = Get bIsUpDirection()
   > Set bIsExageratedPitch(bIsExageratedPitch:bool=OR Boolean(A=AND Boolean(A=@n6, B=@n7), B=AND Boolean(A=@n3, B=@n2))) -> (Output_Get:bool)
   n8 = Get PitchTransitionDown() -> (PitchTransitionDown:Anim Sequence Object Reference)
   n9 = Get PitchTransitionUp() -> (PitchTransitionUp:Anim Sequence Object Reference)
   n10 = Get bIsUpDirection()
   n11 = Get bIsUpDirection()
   n12 = Get PitchTransitionExageratedUp() -> (PitchTransitionExageratedUp:Anim Sequence Object Reference)
   n13 = Get PitchTransitionExageratedDown() -> (PitchTransitionExageratedDown:Anim Sequence Object Reference)
   > SetSequence(
    SequencePlayer=Get SequencePlayer() -> (SequencePlayer:Sequence Player Reference Structure)
    Sequence=Select(
     Option 0:Anim Sequence Object Reference=Select(Option 0=@n8, Option 1=@n9, Index=@n10)
     Option 1:Anim Sequence Object Reference=Select(Option 0=@n13, Option 1=@n12, Index=@n11)
     Index=Get bIsExageratedPitch() -> (bIsExageratedPitch:bool)
    ) -> (ReturnValue:Anim Sequence Object Reference)
   )
   > Return Node()
   Reroute Node(InputPin:Wildcard) -> (OutputPin:Wildcard)
  ]}
 Predicted Landing End Reset {nodes [
   Predicted Landing End Reset!() -> (then, Context:"Anim Update Context Structure (by ref)", Node:"Anim Node Reference Structure (by ref)") {is=title K2Node_FunctionEntry}
   > Set bPredictiveLandingEnd(bPredictiveLandingEnd:bool=false) -> (Output_Get:bool)
   > Return Node()
  ]}
 SecondMovementMode_FinalPoseOverride {nodes [
   AnimGraphNode_StateMachine_1 = PelvisOffset SM() {is=title AnimGraphNode_StateMachine, bSkipFirstUpdateTransition=false}
   Output Pose(Result=Slot 'TDiA.Custom_01'(Source=@AnimGraphNode_StateMachine_1) {slotName=TDiA.Custom_01}) {layerGroup=SecondMovementMode}
   AnimGraphNode_StateMachine_0 = Flying SM() {is=title AnimGraphNode_StateMachine}
   n1 = Apply Additive(
    Base=@AnimGraphNode_StateMachine_0
    Additive=Blendspace Player 'Wingsuit_Gadget_Flying_Pose_Additive'(X=Get Roll() -> (Roll:double), Y=Get Pitch() -> (Pitch:double)) {blendSpace=/WingSuit/Animation/Game/MainPlayer/AnimAssets/WingSuitClothAnim/Wingsuit_Gadget_Flying_Pose_Additive}
    Alpha=Get PoseAlpha() -> (PoseAlpha:double)
   )
   n2 = Blendspace Player 'Wingsuit_PhysicsFlying_BS'(X=Get Roll(), Y=Get Pitch()) {bLoop=false, blendSpace=&/Wingsuit_PhysicsFlying_BS}
   n3 = Get TiltAlpha() -> (TiltAlpha:double)
   Save cached pose 'PrePelvisOffset Pose'(Pose=Blend Poses by bool(
     BlendPose_0=Apply Additive(Base=@n1, Additive=@n2, Alpha=@n3)
     BlendPose_1=Input Pose!() -> (Pose) {name=InFinalPose}
     BlendTime_0=0
     BlendTime_1=0
     bActiveValue=Get bIsFlying() -> (bIsFlying:bool)
    )) {cacheName=PrePelvisOffset Pose}
  ]}
 Flying SM {
  entry=Launch
  path=SecondMovementMode_FinalPoseOverride.AnimGraphNode_StateMachine_0.Flying SM
  ids=id
  nodes [
   Predictive Land Alias() {is=title AnimStateAliasNode, also {aliases [Idle, Launch, Pitch Transition, Yaw Transition]}}
   Flip Alias() {is=title AnimStateAliasNode, also {aliases [Idle, Launch, Pitch Transition, Predictive Land, Yaw Transition]}}
   AnimStateNode_0 = Idle()
   AnimStateNode_1 = Pitch Transition() {is=title AnimStateNode}
   AnimStateNode_2 = Yaw Transition() {is=title AnimStateNode}
   AnimStateNode_3 = Predictive Land() {is=title AnimStateNode}
   AnimStateNode_4 = Launch() {is=title AnimStateNode}
   AnimStateNode_5 = Flip() {is=title AnimStateNode}
   AnimStateTransitionNode_0 = AnimStateTransitionNode() {crossfadeDuration=0.35}
   AnimStateTransitionNode_1 = AnimStateTransitionNode() {bAllowInertializationForSelfTransitions=true, crossfadeDuration=0.35}
   AnimStateTransitionNode_2 = AnimStateTransitionNode() {bAutomaticRuleBasedOnSequencePlayerInState=true, crossfadeDuration=0.5}
   AnimStateTransitionNode_3 = AnimStateTransitionNode() {bAllowInertializationForSelfTransitions=true, bAutomaticRuleBasedOnSequencePlayerInState=true, crossfadeDuration=0.5}
   AnimStateTransitionNode_4 = AnimStateTransitionNode() {bAllowInertializationForSelfTransitions=true, bAutomaticRuleBasedOnSequencePlayerInState=true, crossfadeDuration=0.35}
   AnimStateTransitionNode_5 = AnimStateTransitionNode() {bAllowInertializationForSelfTransitions=true, crossfadeDuration=0.3}
   AnimStateTransitionNode_6 = AnimStateTransitionNode() {bAllowInertializationForSelfTransitions=true, crossfadeDuration=0.4}
   AnimStateTransitionNode_7 = AnimStateTransitionNode() {bAllowInertializationForSelfTransitions=true, crossfadeDuration=0.35}
   AnimStateTransitionNode_8 = AnimStateTransitionNode() {bAllowInertializationForSelfTransitions=true, bAutomaticRuleBasedOnSequencePlayerInState=true, crossfadeDuration=0.35}
  ]
  links [
   {
    dst=Idle
    rule=SecondMovementMode_FinalPoseOverride.AnimGraphNode_StateMachine_0.Flying SM.AnimStateTransitionNode_8.Transition
    src=Flip
    via=AnimStateTransitionNode_8
   }
   {
    dst=Flip
    rule=SecondMovementMode_FinalPoseOverride.AnimGraphNode_StateMachine_0.Flying SM.AnimStateTransitionNode_7.Transition
    src=Flip Alias
    via=AnimStateTransitionNode_7
   }
   {
    dst=Pitch Transition
    rule=SecondMovementMode_FinalPoseOverride.AnimGraphNode_StateMachine_0.Flying SM.AnimStateTransitionNode_0.Transition
    src=Idle
    via=AnimStateTransitionNode_0
   }
   {
    dst=Yaw Transition
    rule=SecondMovementMode_FinalPoseOverride.AnimGraphNode_StateMachine_0.Flying SM.AnimStateTransitionNode_1.Transition
    src=Idle
    via=AnimStateTransitionNode_1
   }
   {
    dst=Idle
    rule=SecondMovementMode_FinalPoseOverride.AnimGraphNode_StateMachine_0.Flying SM.AnimStateTransitionNode_4.Transition
    src=Launch
    via=AnimStateTransitionNode_4
   }
   {
    dst=Idle
    rule=SecondMovementMode_FinalPoseOverride.AnimGraphNode_StateMachine_0.Flying SM.AnimStateTransitionNode_2.Transition
    src=Pitch Transition
    via=AnimStateTransitionNode_2
   }
   {
    dst=Idle
    rule=SecondMovementMode_FinalPoseOverride.AnimGraphNode_StateMachine_0.Flying SM.AnimStateTransitionNode_6.Transition
    src=Predictive Land
    via=AnimStateTransitionNode_6
   }
   {
    dst=Predictive Land
    rule=SecondMovementMode_FinalPoseOverride.AnimGraphNode_StateMachine_0.Flying SM.AnimStateTransitionNode_5.Transition
    src=Predictive Land Alias
    via=AnimStateTransitionNode_5
   }
   {
    dst=Idle
    rule=SecondMovementMode_FinalPoseOverride.AnimGraphNode_StateMachine_0.Flying SM.AnimStateTransitionNode_3.Transition
    src=Yaw Transition
    via=AnimStateTransitionNode_3
   }
  ]
 }
 Idle {
  path=SecondMovementMode_FinalPoseOverride.AnimGraphNode_StateMachine_0.Flying SM.AnimStateNode_0.Idle
  nodes [
   n1 = Apply Additive(
    Base=Sequence Player 'Wingsuit_Flying_pose_neutral__WS'() {bLoopAnimation=false, sequence=/WingSuit/Animation/Game/MainPlayer/AnimAssets/WingSuitClothAnim/Wingsuit_Flying_pose_neutral__WS}
    Additive="Sequence Player (additive) 'Wingsuit_Flying_SlowDecent__WS'"() {sequence=/WingSuit/Animation/Game/MainPlayer/AnimAssets/WingSuitClothAnim/Wingsuit_Flying_SlowDecent__WS}
   )
   n2 = "Sequence Player (additive) 'Wingsuit_Flying_IdleCycle__WS'"() {sequence=/WingSuit/Animation/Game/MainPlayer/AnimAssets/WingSuitClothAnim/Wingsuit_Flying_IdleCycle__WS}
   n3 = Sequence Player 'Wingsuit_Flying_pose_neutral__WS'() {bLoopAnimation=false, sequence=&Wingsuit_Flying_pose_neutral__WS}
   Output Animation Pose(Result=Blend Poses by bool(
     BlendPose_0=Modify Curve(SourcePose=@n1, CurveValues_0=1, Alpha=1) {curveNames [posealpha], curveValues [0]}
     BlendPose_1=Apply Additive(Base=@n3, Additive=@n2)
     BlendTime_0=0.6
     BlendTime_1=0.4
     bActiveValue=Get bIsSlowDecent() -> (bIsSlowDecent:bool)
    )) {
    node {stateEntryFunction {functionName=OnStateEntry}}
    stateEntryFunction {bSelfContext=true, memberGuid=#1, memberName=Predicted Landing End Reset}
   }
  ]
 }
 Pitch Transition {
  path=SecondMovementMode_FinalPoseOverride.AnimGraphNode_StateMachine_0.Flying SM.AnimStateNode_1.Pitch Transition
  nodes [
   n1 = Sequence Player 'Wingsuit_Fly_Transition_Up_B__WS'() {
    becomeRelevantFunction {bSelfContext=true, memberGuid=#2, memberName=PitchTransitionSequence_OnBecomeRelevant}
    node {
     bLoopAnimation=false
     sequence=/WingSuit/Animation/Game/MainPlayer/AnimAssets/WingSuitClothAnim/Wingsuit_Fly_Transition_Up_B__WS
     becomeRelevantFunction {functionName=OnBecomeRelevant}
    }
   }
   Output Animation Pose(Result=Modify Curve(SourcePose=@n1, CurveValues_0=1, Alpha=1) {curveNames [posealpha], curveValues [0]})
  ]
 }
 Yaw Transition {
  path=SecondMovementMode_FinalPoseOverride.AnimGraphNode_StateMachine_0.Flying SM.AnimStateNode_2.Yaw Transition
  nodes [
   Blend Poses by bool(
    BlendPose_0=Sequence Player 'Wingsuit_Fly_Transition_Right_B__WS'() {bLoopAnimation=false, sequence=/WingSuit/Animation/Game/MainPlayer/AnimAssets/WingSuitClothAnim/Wingsuit_Fly_Transition_Right_B__WS}
    BlendPose_1=Sequence Player 'Wingsuit_Flying_TurnNoiseA__WS'() {bLoopAnimation=false, sequence=/WingSuit/Animation/Game/MainPlayer/AnimAssets/WingSuitClothAnim/Wingsuit_Flying_TurnNoiseA__WS}
    bActiveValue=Get bRandomYaw() -> (bRandomYaw:bool)
   )
   Blend Poses by bool(
    BlendPose_0=Sequence Player 'Wingsuit_Fly_Transition_Left_B__WS'() {bLoopAnimation=false, sequence=/WingSuit/Animation/Game/MainPlayer/AnimAssets/WingSuitClothAnim/Wingsuit_Fly_Transition_Left_B__WS}
    BlendPose_1=Sequence Player 'Wingsuit_Flying_TurnNoiseA__WS'() {bLoopAnimation=false, sequence=&Wingsuit_Flying_TurnNoiseA__WS}
    bActiveValue=Get bRandomYaw()
   )
   n1 = Sequence Player 'Wingsuit_Fly_Transition_Left_B__WS'() {bLoopAnimation=false, sequence=&Wingsuit_Fly_Transition_Left_B__WS}
   n2 = Sequence Player 'Wingsuit_Fly_Transition_Right_B__WS'() {bLoopAnimation=false, sequence=&Wingsuit_Fly_Transition_Right_B__WS}
   n3 = Get bIsLeftDirection() -> (bIsLeftDirection:bool)
   Output Animation Pose(Result=Modify Curve(SourcePose=Blend Poses by bool(BlendPose_0=@n1, BlendPose_1=@n2, bActiveValue=@n3), CurveValues_0=1, Alpha=1) {curveNames [posealpha], curveValues [0]}) {
    node {stateExitFunction {functionName=Yaw Finished}}
    stateExitFunction {bSelfContext=true, memberGuid=#3, memberName=Yaw Finished}
   }
  ]
 }
 Predictive Land {
  path=SecondMovementMode_FinalPoseOverride.AnimGraphNode_StateMachine_0.Flying SM.AnimStateNode_3.Predictive Land
  nodes [Output Animation Pose(Result=Modify Curve!(
     SourcePose=Sequence Player 'Wingsuit_Landing_VerticalNoRoll__WS'() {bLoopAnimation=false, sequence=/WingSuit/Animation/Game/MainPlayer/AnimAssets/WingSuitClothAnim/Wingsuit_Landing_VerticalNoRoll__WS}
     CurveValues_0=1
     CurveValues_1:float=1
     Alpha=1
    ) -> (Pose) {curveNames [tiltalpha,posealpha], curveValues [0,0]})]
 }
 Launch {
  path=SecondMovementMode_FinalPoseOverride.AnimGraphNode_StateMachine_0.Flying SM.AnimStateNode_4.Launch
  nodes [
   n1 = Sequence Player 'Wingsuit_Fly_Launch_Boost__WS'!(StartPosition:float=0.777) -> (Pose) {
    becomeRelevantFunction {bSelfContext=true, memberGuid=#4, memberName=Launch_OnBecomeRelevant}
    node {
     bLoopAnimation=false
     sequence=/WingSuit/Animation/Game/MainPlayer/AnimAssets/WingSuitClothAnim/Wingsuit_Fly_Launch_Boost__WS
     becomeRelevantFunction {functionName=OnBecomeRelevant}
    }
   }
   Output Animation Pose(Result=@n1)
  ]
 }
 Flip {
  path=SecondMovementMode_FinalPoseOverride.AnimGraphNode_StateMachine_0.Flying SM.AnimStateNode_5.Flip
  nodes [
   n1 = Sequence Player 'Wingsuit_Flying_BackflipTransitionDown__WS'() {bLoopAnimation=false, sequence=/WingSuit/Animation/Game/MainPlayer/AnimAssets/WingSuitClothAnim/Wingsuit_Flying_BackflipTransitionDown__WS}
   Output Animation Pose(Result=@n1)
  ]
 }
 Transition {
  path=SecondMovementMode_FinalPoseOverride.AnimGraphNode_StateMachine_0.Flying SM.AnimStateTransitionNode_0.Transition
  nodes [Result(bCanEnterTransition=Get bIsPitchTransition() -> (bIsPitchTransition:bool))]
 }
 Transition {
  path=SecondMovementMode_FinalPoseOverride.AnimGraphNode_StateMachine_0.Flying SM.AnimStateTransitionNode_1.Transition
  nodes [Result(bCanEnterTransition=Get bIsYawTransition() -> (bIsYawTransition:bool))]
 }
 Transition {path=SecondMovementMode_FinalPoseOverride.AnimGraphNode_StateMachine_0.Flying SM.AnimStateTransitionNode_2.Transition, nodes [Result()]}
 Transition {path=SecondMovementMode_FinalPoseOverride.AnimGraphNode_StateMachine_0.Flying SM.AnimStateTransitionNode_3.Transition, nodes [Result()]}
 Transition {path=SecondMovementMode_FinalPoseOverride.AnimGraphNode_StateMachine_0.Flying SM.AnimStateTransitionNode_4.Transition, nodes [Result()]}
 Transition {
  path=SecondMovementMode_FinalPoseOverride.AnimGraphNode_StateMachine_0.Flying SM.AnimStateTransitionNode_5.Transition
  nodes [Result(bCanEnterTransition=Get bPredictiveLandingStart() -> (bPredictiveLandingStart:bool))]
 }
 Transition {
  path=SecondMovementMode_FinalPoseOverride.AnimGraphNode_StateMachine_0.Flying SM.AnimStateTransitionNode_6.Transition
  nodes [
   Result(bCanEnterTransition=Get bPredictiveLandingEnd() -> (bPredictiveLandingEnd:bool))
   NOT Boolean(A=Get bPredictiveLanding() -> (bPredictiveLanding:bool))
  ]
 }
 Transition {
  path=SecondMovementMode_FinalPoseOverride.AnimGraphNode_StateMachine_0.Flying SM.AnimStateTransitionNode_7.Transition
  nodes [Result(bCanEnterTransition=Get bFlipping() -> (bFlipping:bool))]
 }
 Transition {path=SecondMovementMode_FinalPoseOverride.AnimGraphNode_StateMachine_0.Flying SM.AnimStateTransitionNode_8.Transition, nodes [Result()]}
 PelvisOffset SM {
  entry=PelvisOffset None Start_1
  path=SecondMovementMode_FinalPoseOverride.AnimGraphNode_StateMachine_1.PelvisOffset SM
  ids=id
  nodes [
   AnimStateNode_0 = PelvisOffset Down_1() {is=title AnimStateNode}
   AnimStateNode_1 = PelvisOffset None Start_1() {is=title AnimStateNode}
   AnimStateNode_3 = PelvisOffset None Predictive Landing_1() {is=title AnimStateNode}
   AnimStateTransitionNode_0 = AnimStateTransitionNode() {bAllowInertializationForSelfTransitions=true}
   AnimStateTransitionNode_4 = AnimStateTransitionNode() {bAllowInertializationForSelfTransitions=true, crossfadeDuration=0.3}
   AnimStateTransitionNode_5 = AnimStateTransitionNode() {bAllowInertializationForSelfTransitions=true, crossfadeDuration=0.4}
   AnimStateTransitionNode_6 = AnimStateTransitionNode() {bAllowInertializationForSelfTransitions=true, crossfadeDuration=0.4}
  ]
  links [
   {
    dst=PelvisOffset None Predictive Landing_1
    rule=SecondMovementMode_FinalPoseOverride.AnimGraphNode_StateMachine_1.PelvisOffset SM.AnimStateTransitionNode_4.AnimationTransitionGraph_0
    src=PelvisOffset Down_1
    via=AnimStateTransitionNode_4
   }
   {
    dst=PelvisOffset None Start_1
    rule=SecondMovementMode_FinalPoseOverride.AnimGraphNode_StateMachine_1.PelvisOffset SM.AnimStateTransitionNode_0.AnimationTransitionGraph_0
    src=PelvisOffset Down_1
    via=AnimStateTransitionNode_0
   }
   {
    dst=PelvisOffset Down_1
    rule=SecondMovementMode_FinalPoseOverride.AnimGraphNode_StateMachine_1.PelvisOffset SM.AnimStateTransitionNode_5.AnimationTransitionGraph_0
    src=PelvisOffset None Predictive Landing_1
    via=AnimStateTransitionNode_5
   }
   {
    dst=PelvisOffset Down_1
    rule=SecondMovementMode_FinalPoseOverride.AnimGraphNode_StateMachine_1.PelvisOffset SM.AnimStateTransitionNode_6.AnimationTransitionGraph_0
    src=PelvisOffset None Start_1
    via=AnimStateTransitionNode_6
   }
  ]
 }
 PelvisOffset Down_1 {
  path=SecondMovementMode_FinalPoseOverride.AnimGraphNode_StateMachine_1.PelvisOffset SM.AnimStateNode_0.PelvisOffset Down_1
  nodes [
   n1 = Local To Component(LocalPose=Use cached pose 'PrePelvisOffset Pose'())
   Output Animation Pose(Result=Component To Local(ComponentPose="Transform (Modify) Bone - Bone: pelvis"(ComponentPose=@n1, Alpha=1, Translation="0,0,-60", Rotation="0,0,0", Scale="1,1,1") {translationMode=BMM_Additive, translationSpace=BCS_WorldSpace, boneToModify {boneName=pelvis}, translation {z=-60}}))
  ]
 }
 PelvisOffset None Start_1 {
  path=SecondMovementMode_FinalPoseOverride.AnimGraphNode_StateMachine_1.PelvisOffset SM.AnimStateNode_1.PelvisOffset None Start_1
  nodes [Output Animation Pose(Result=Use cached pose 'PrePelvisOffset Pose'())]
 }
 PelvisOffset None Predictive Landing_1 {
  path=SecondMovementMode_FinalPoseOverride.AnimGraphNode_StateMachine_1.PelvisOffset SM.AnimStateNode_3.PelvisOffset None Predictive Landing_1
  nodes [Output Animation Pose(Result=Use cached pose 'PrePelvisOffset Pose'())]
 }
 AnimationTransitionGraph_0 {
  path=SecondMovementMode_FinalPoseOverride.AnimGraphNode_StateMachine_1.PelvisOffset SM.AnimStateTransitionNode_0.AnimationTransitionGraph_0
  nodes [Result(bCanEnterTransition=NOT Boolean(A=Get bIsFlying()))]
 }
 AnimationTransitionGraph_0 {
  path=SecondMovementMode_FinalPoseOverride.AnimGraphNode_StateMachine_1.PelvisOffset SM.AnimStateTransitionNode_4.AnimationTransitionGraph_0
  nodes [Result(bCanEnterTransition=Get bPredictiveLandingStart())]
 }
 AnimationTransitionGraph_0 {
  path=SecondMovementMode_FinalPoseOverride.AnimGraphNode_StateMachine_1.PelvisOffset SM.AnimStateTransitionNode_5.AnimationTransitionGraph_0
  nodes [Result(bCanEnterTransition=Get bPredictiveLandingEnd())]
 }
 AnimationTransitionGraph_0 {
  path=SecondMovementMode_FinalPoseOverride.AnimGraphNode_StateMachine_1.PelvisOffset SM.AnimStateTransitionNode_6.AnimationTransitionGraph_0
  nodes [Result(bCanEnterTransition=Get bIsFlying())]
 }
 TransitionGraph {links [], nodes []}
 Yaw Finished {nodes [
   Yaw Finished!() -> (then, Context:"Anim Update Context Structure (by ref)", Node:"Anim Node Reference Structure (by ref)") {is=title K2Node_FunctionEntry}
   > Set bIsYawTransitionFinished(bIsYawTransitionFinished:bool=true) -> (Output_Get:bool)
   > Return Node()
  ]}
]
variables [
 IsWingSuitMeshOn {type=boolean}
 LaunchSequenceIndex {type=number}
 PitchTransitionDown {type=AnimSequence, default=/WingSuit/Animation/Game/MainPlayer/AnimAssets/WingSuitClothAnim/Wingsuit_Fly_Transition_Down_A__WS}
 PitchTransitionExageratedDown {type=AnimSequence, default=/WingSuit/Animation/Game/MainPlayer/AnimAssets/WingSuitClothAnim/Wingsuit_Fly_Transition_Down_B__WS}
 PitchTransitionExageratedUp {type=AnimSequence, default=&Wingsuit_Fly_Transition_Up_B__WS}
 PitchTransitionUp {type=AnimSequence, default=/WingSuit/Animation/Game/MainPlayer/AnimAssets/WingSuitClothAnim/Wingsuit_Fly_Transition_Up_A__WS}
]
